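# Translucent glass material for the TulipGlory backpack that fakes a lit, liquid-filled container with a "Facehugger" visible inside.
# Scene-minus-pixel depth builds Facehugger masks and fog, while world-normal, bounds and Fresnel terms shape lighting, liquid darkening and glass edges.
# Emissive blends tinted scene color, gradient liquid, decal and TechArtMask-driven moss over a shared glass base.
# Quality switches drop the scene-color blend on mobile; static switches add camera dither and vertex hiding.
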Material /BRCosmetics/Accessories/FORT_Backpacks/Backpack_M_MED_TulipGlory/Materials/M_TulipGlory_Backpack_Glass
graph {
 properties {
  bAutomaticallySetUsageInEditor=false
  bEnableResponsiveAA=true
  bScreenSpaceReflections=true
  bUseMaterialAttributes=true
  bUsedWithMeshDeformer=true
  bUsedWithSkeletalMesh=true
  blendMode=BLEND_Translucent
  translucencyDirectionalLightingIntensity=5
  translucencyLightingMode=TLM_SurfacePerPixelLighting
  usedShadingModels=Default Lit
 }
 nodes [
  n1 = ScalarParameter() {defaultValue=0.424, group=TulipGlory - Backpack, parameterName=Light Influence - Shape Size}
  n2 = Add(A=@n1, B=ScalarParameter() {defaultValue=0.8734, group=TulipGlory - Backpack, parameterName=Light Influence - Smoothness})
  n3 = Transform(None=SceneTexture() {sceneTextureId=PPI_WorldNormal}.Color) {transformSourceType=TRANSFORMSOURCE_World, transformType=TRANSFORM_Local}
  n4 = VectorParameter() {group=TulipGlory - Backpack, parameterName=Light Influence - Direction, defaultValue=<rgba -0.35 -0.35 -1 1>}
  n5 = DotProduct(A=@n3, B=@n4.RGB)
  n6 = ScalarParameter() {defaultValue=0.3973, group=TulipGlory - Backpack, parameterName=Light Influence - Height}
  n7 = SmoothStep(
   Min=@n6
   Max=Add(A=@n6, B=ScalarParameter() {defaultValue=0.5724, group=TulipGlory - Backpack, parameterName=Light Influence - Height Smoothness})
   Value=NamedRerouteUsage()
  )
  NamedRerouteDeclaration(None=Multiply(A=SmoothStep(Min=@n1, Max=@n2, Value=@n5), B=@n7)) {name=World Normal Mask, nodeColor=<rgba 0.8516 1 0 1>}
  n8 = Multiply(
   A=Subtract(A=SceneDepth().R, B=PixelDepth().R)
   B=ScalarParameter() {defaultValue=0.0029, group=TulipGlory - Backpack, parameterName=Facehugger - Depth}
  )
  n9 = Step(Y=@n8, X=ScalarParameter() {defaultValue=0.0176, group=TulipGlory - Backpack, parameterName=Facehugger - Step})
  NamedRerouteDeclaration(None=Multiply(A=@n9, B=NamedRerouteUsage())) {name=Light Mask, nodeColor=<rgba 1 0.1875 0 1>}
  NamedRerouteDeclaration(None=@n9) {name=Stepped Facehugger Mask, nodeColor=<rgba 1 0 0.0703 1>}
  n10 = NamedRerouteUsage()
  n11 = MaterialFunctionCall!(
   Distance Bias=Constant()
   Distance Divisor=LinearInterpolate(
    A=ScalarParameter() {defaultValue=22}
    B=ScalarParameter() {defaultValue=4, group=TulipGlory - Backpack, parameterName=Fog - Distance}
    Alpha=Saturate(None=@n10)
   )
   Scene Depth
   Pixel Depth
  ) {materialFunction=/Game/Effects/Fort_Effects/Materials/_Functions/MF_DistanceBehindTranslucency_01}
  n12 = ScalarParameter() {defaultValue=0.15, group=TulipGlory - Backpack, parameterName=Fog - Power}
  NamedRerouteDeclaration(None=Saturate(None=Power(Base=@n11.Normalized, Exp=@n12))) {name=Fog Depth, nodeColor=<rgba 0 0.3906 1 1>}
  n13 = ScalarParameter() {defaultValue=0.35, group=TulipGlory - Backpack, parameterName=Source Light - Height}
  n14 = Add(A=@n13, B=ScalarParameter() {defaultValue=4.5, group=TulipGlory - Backpack, parameterName=Source Light - Smoothnesss})
  n15 = NamedRerouteUsage()
  n16 = ScalarParameter() {defaultValue=1.1739, group=TulipGlory - Backpack, parameterName=Source Light - Fresnel Shape}
  n17 = ScalarParameter() {defaultValue=5.6499, group=TulipGlory - Backpack, parameterName=Source Light - Peak Shape}
  n18 = ScalarParameter() {defaultValue=0.618, group=TulipGlory - Backpack, parameterName=Source Light - Peak Height}
  n19 = Add(A=@n18, B=ScalarParameter() {defaultValue=1.2748, group=TulipGlory - Backpack, parameterName=Source Light - Peak Smoothness})
  n20 = NamedRerouteUsage()
  n21 = Add(
   A=Multiply(
    A=SmoothStep(Min=@n13, Max=@n14, Value=@n15)
    B=MaterialFunctionCall(Exponent=@n16) {materialFunction=&AxisAlignedFresnel}.Fresnel
   )
   B=Multiply(
    A=MaterialFunctionCall(Exponent=@n17) {materialFunction=&AxisAlignedFresnel}.Fresnel
    B=SmoothStep(Min=@n18, Max=@n19, Value=@n20)
   )
  )
  n22 = Multiply(
   A=LinearInterpolate(A=@n21, B=Multiply(A=@n21) {constB=0, desc=Use Facehugger mask to make it look backlit}, Alpha=NamedRerouteUsage())
   B=ScalarParameter() {defaultValue=100, group=TulipGlory - Backpack, parameterName=Source Light - Intensity}
  )
  n23 = Multiply(A=NamedRerouteUsage(), B=ScalarParameter() {defaultValue=5, group=TulipGlory - Backpack, parameterName=Facehugger - Light Influence})
  NamedRerouteDeclaration(None=Max(A=Add(A=@n22, B=@n23))) {name=Source Light, nodeColor=<rgba 0.8984 1 0 1>}
  n24 = ScalarParameter() {defaultValue=0.5928, group=TulipGlory - Backpack, parameterName="Edge -  Step"}
  n25 = Add(A=@n24, B=ScalarParameter() {defaultValue=1.1762, group=TulipGlory - Backpack, parameterName="Edge -  Smoothness"})
  n26 = Sine(None=Fresnel())
  n27 = Multiply(
   A=Multiply(A=SmoothStep(Min=@n24, Max=@n25, Value=@n26), B=NamedRerouteUsage())
   B=VectorParameter() {group=TulipGlory - Backpack, parameterName=Fresnel Color, defaultValue=<rgba 0.491 0.4452 0.162 1>}.RGB
  )
  n28 = ScalarParameter() {defaultValue=1, group=TulipGlory - Backpack, parameterName=Fresnel - Intensity}
  NamedRerouteDeclaration(None=Saturate(None=Multiply(A=@n27, B=@n28))) {name=Glass Edge, nodeColor=<rgba 1 0 0.2812 1>}
  n29 = ScalarParameter() {defaultValue=0.0072, parameterName=Facehugger - Side Light Length}
  n30 = ScalarParameter() {defaultValue=0.0248, group=TulipGlory - Backpack, parameterName=Facehugger - Smoothness}
  n31 = Add(A=@n29, B=@n30)
  NamedRerouteDeclaration(None=OneMinus(None=SmoothStep(Min=@n29, Max=@n31, Value=@n8))) {name=Side Light Mask, nodeColor=<rgba 0.3359 0 1 1>}
  n32 = LocalPosition() {localOrigin=InstancePreSkinning}
  n33 = OneMinus(None=ComponentMask(None=VertexInterpolator(VS=@n32.XYZ).PS) {b=true})
  n34 = PreSkinnedLocalBounds()
  n35 = ComponentMask(None=@n34.Half Extents) {b=true}
  NamedRerouteDeclaration(None=Divide(A=Add(A=@n33, B=@n35) {constB=100}, B=ComponentMask(None=@n34.Extents) {b=true})) {name=Preskinned Bounds, nodeColor=<rgba 0.4062 0 1 1>}
  n36 = OneMinus(None=@n7)
  n37 = Saturate(None=DotProduct(A=@n3, B=VectorParameter() {defaultValue=<rgba -1 0 0 0>}.RGB))
  n38 = NamedRerouteUsage()
  n39 = ScalarParameter() {defaultValue=2, parameterName=Side Light Mask - Intensity}
  NamedRerouteDeclaration(None=Multiply(A=Multiply(A=@n36, B=@n37), B=Multiply(A=@n38, B=@n39))) {name=test, nodeColor=<rgba 0 0.3906 1 1>}
  n40 = ScalarParameter() {defaultValue=-0.0006, group=TulipGlory - Backpack, parameterName=Facehugger - Smooth Step}
  n41 = Add(A=@n40, B=@n30)
  NamedRerouteDeclaration(None=OneMinus(None=SmoothStep(Min=@n40, Max=@n41, Value=@n8))) {name=Facehugger Mask, nodeColor=<rgba 0.3359 1 0 1>}
  n42 = TextureSampleParameter2D() {
   group=00_Glass
   parameterName=TechArtMask
   samplerType=SAMPLERTYPE_Masks
   sortPriority=-96
   texture=&T_BlackMask
  }
  NamedRerouteDeclaration(None=@n42.R) {name=Moss, nodeColor=<rgba 0 0.3906 1 1>}
  NamedRerouteDeclaration(None=@n42.G) {name=Glass Opacity, nodeColor=<rgba 1 0.4219 0 1>}
  n43 = NamedRerouteUsage()
  n44 = ScalarParameter() {defaultValue=0.655, group=TulipGlory - Backpack, parameterName=Moss - Exponent}
  n45 = Multiply(
   A=Multiply(A=Power(Base=@n43, Exp=@n44), B=ScalarParameter() {defaultValue=1.5495, group=TulipGlory - Backpack, parameterName=Moss - Intensity})
   B=NamedRerouteUsage()
  )
  n46 = NamedRerouteUsage()
  n47 = ScalarParameter() {defaultValue=0.5305, group=TulipGlory - Backpack, parameterName=Moss - Step Removal}
  n48 = Saturate(None=OneMinus(None=Step(Y=@n46, X=@n47) {constX=0.984}))
  NamedRerouteDeclaration(None=Saturate(None=Multiply(A=@n45, B=@n48))) {name=Moss Mask, nodeColor=<rgba 0.5469 1 0 1>}
  NamedRerouteUsage()
  NamedRerouteUsage()
  ScalarParameter() {defaultValue=2.5615, group=TulipGlory - Backpack, parameterName=Source Light - Smoothness}
  n49 = MaterialFunctionCall!(
   Material Attributes=MaterialFunctionCall!(TechArtMask=@n42.RGB) {materialFunction=/Game/Global/Materials/Character_Parent_2023/Glass_Parent_2023/MF_Glass}.Result
   Fresnel=Fresnel() {baseReflectFraction=0, exponent=1}
   Alpha=@n42.RGB
  ) {materialFunction=/Game/Global/MaterialFunctions/TechArt/MF_ThinFilm_TechArt}
  n50 = StaticSwitchParameter(
   True=MaterialFunctionCall!(In=@n49.Result) {materialFunction=/Game/Packages/Fortress_SharedMaterials/Base_Material_Functions/BaseCustomizationFunctions/MF_FN_Vehicle_CameraDither}.Result
   False=@n49.Result
  ) {group=00_Glass, parameterName=UseCameraDither, sortPriority=250}
  n51 = ObjectPositionWS()
  n52 = WorldPosition()
  n53 = SetMaterialAttributes!(
   MaterialAttributes=@n50
   World Position Offset=Multiply(A=Subtract(A=@n51, B=@n52.XYZ), B=ScalarParameter() {group=00_Glass, parameterName=VertexHide, sortPriority=301})
  ) {attributeSetTypes [#1]}
  n54 = MaterialFunctionCall!(Material=@n50) {materialFunction=/Game/Global/Materials/Character_Parent_2023/Functions/BaseFN_CharacterOpacity_2023}
  n55 = SceneColor!(UVs)
  n56 = ScalarParameter() {group=TulipGlory - Backpack, parameterName=Facehugger BG - Intensity}
  n57 = ScalarParameter() {defaultValue=0.3, group=TulipGlory - Backpack, parameterName=Facehugger - Intensity}
  n58 = Step(
   Y=OneMinus(None=MaterialFunctionCall() {materialFunction=&AxisAlignedFresnel}.Fresnel)
   X=ScalarParameter() {defaultValue=0.6706, group=TulipGlory - Backpack, parameterName=Glass Fresnel - Size}
  )
  n59 = TextureSample(UVs=Saturate(None=NamedRerouteUsage())) {samplerType=SAMPLERTYPE_LinearColor, texture=&/CA_TulipGlory}
  n60 = Multiply(A=Multiply(A=@n59.RGB, B=NamedRerouteUsage()), B=NamedRerouteUsage())
  n61 = VectorParameter() {parameterName=Param_1, defaultValue=<rgba 0.4971 0.7632 1 1>}
  n62 = Multiply(
   A=Max(
    A=NamedRerouteDeclaration(None=@n58) {name=Fresnel Edge Removal, nodeColor=<rgba 0 0.5781 1 1>}
    B=ScalarParameter() {defaultValue=0.5, group=TulipGlory - Backpack, parameterName=Glass Fresnel - Min Opacity}
   )
   B=LinearInterpolate(A=@n60, B=Multiply(A=@n60, B=@n61.RGB), Alpha=NamedRerouteUsage())
  )
  n63 = ScalarParameter() {defaultValue=0.0211, group=TulipGlory - Backpack, parameterName=Liquid Darkening - Step}
  n64 = ScalarParameter() {defaultValue=0.856, group=TulipGlory - Backpack, parameterName=Liquid Darkening - Smoothness}
  n65 = NamedRerouteUsage()
  n66 = ScalarParameter() {defaultValue=1.5939, group=TulipGlory - Backpack, parameterName=Liquid Darkening - Fresnel}
  n67 = Multiply(
   A=SmoothStep(Min=@n63, Max=Add(A=@n63, B=@n64), Value=OneMinus(None=@n65))
   B=OneMinus(None=MaterialFunctionCall(Exponent=@n66) {materialFunction=&AxisAlignedFresnel}.Fresnel)
  )
  n68 = ScalarParameter() {defaultValue=2, group=TulipGlory - Backpack, parameterName=Liquid Darkening - Intensity}
  n69 = LinearInterpolate(A=@n62, B=Multiply(A=@n62) {constB=0}, Alpha=Saturate(None=Multiply(A=@n67, B=@n68)))
  n70 = LinearInterpolate(
   A=LinearInterpolate(A=Multiply(A=@n55.RGB, B=@n56), B=Multiply(A=@n55.RGB, B=@n57), Alpha=NamedRerouteUsage())
   B=@n69
   Alpha=NamedRerouteUsage()
  )
  n71 = MaterialFunctionCall!(
   Low _ PC Low and Med=@n70
   "High _ XB1, PS4"=@n70
   "Epic _ PC High, XSX, PS5"=@n70
   --------
   "Mobile Low _ PC Perf, Android"=@n69
   Mobile Med _ Switch=@n69
  ) {materialFunction=/Game/Global/MaterialFunctions/TechArt/MF_Fortnite_QualitySwitch_Vector3}
  n72 = ScalarParameter() {defaultValue=1.2, group=TulipGlory - Backpack, parameterName=Liquid - Intensity}
  n73 = NamedRerouteUsage()
  n74 = ScalarParameter() {defaultValue=1.6334, group=TulipGlory - Backpack, parameterName=Decal - Opacity}
  n75 = LinearInterpolate(
   A=Add(A=Multiply(A=@n71.Result, B=@n72), B=NamedRerouteUsage())
   B=VectorParameter() {group=TulipGlory - Backpack, parameterName=Decal - Colour}.RGB
   Alpha=Saturate(None=Multiply(A=@n73, B=@n74))
  ) {constB=0}
  n76 = LinearInterpolate(A=@n75, Alpha=NamedRerouteUsage()) {constB=0}
  n77 = ScalarParameter() {defaultValue=0.5, group=TulipGlory - Backpack, parameterName=Facehugger - SSS Amount}
  n78 = ScalarParameter() {defaultValue=1, group=TulipGlory - Backpack, parameterName=Facehugger - Opacity}
  n79 = Constant() {r=0.8}
  SetMaterialAttributes!(
   MaterialAttributes=MaterialFunctionCall(In=StaticSwitchParameter(True=@n53, False=@n54.Result) {group=00_Glass, parameterName=UseVertexHide, sortPriority=300}) {materialFunction=&MF_TechArtOverrides}.Result
   Emissive Color=LinearInterpolate(
    A=@n75
    B=VectorParameter() {group=TulipGlory - Backpack, parameterName=Moss - Colour, defaultValue=<rgba 0.161 0.1286 0.0407 1>}.RGB
    Alpha=NamedRerouteUsage()
   )
   Subsurface Color=LinearInterpolate(A=@n76, B=Multiply(A=@n76, B=@n77), Alpha=NamedRerouteUsage()) {constB=0}
   Opacity=LinearInterpolate(
    A=ScalarParameter() {defaultValue=1, group=TulipGlory - Backpack, parameterName=Glass - Opacity}
    B=MaterialFunctionCall!(
     Low _ PC Low and Med=@n78
     "High _ XB1, PS4"=@n78
     "Epic _ PC High, XSX, PS5"=@n78
     --------
     "Mobile Low _ PC Perf, Android"=@n79
     Mobile Med _ Switch=@n79
    ) {materialFunction=/Game/Global/MaterialFunctions/TechArt/MF_Fortnite_QualitySwitch_Scalar}.Result
    Alpha=NamedRerouteUsage()
   )
   Base Color=Multiply(A=NamedRerouteUsage(), B=ScalarParameter() {defaultValue=0.0413, group=TulipGlory - Backpack, parameterName=Number - Colour})
   Normal
  ) {attributeSetTypes [
    #2
    #3
    #4
    #5
    #6
   ]}
 ]
}
