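# Physics asset for the body mesh of the MED TulipGlory backpack cosmetic.
# It has two capsule bodies, a larger one on Body_01 and a very small one on the pack bone.
# A single constraint joins them with both swing axes locked, leaving twist as the only angular axis not locked, with a narrow twist limit angle set.
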PhysicsAsset /BRCosmetics/Accessories/FORT_Backpacks/Backpack_M_MED_TulipGlory/Meshes/M_MED_TulipGlory_Pack_Body
bodies {
 Body_01 [{
   shapeType=Capsule
   length=4.4553
   radius=3.4134
   shapeName=body_capsule
   center=<0.1461 4.3147 -0.0239>
   rotation=<rot 1.031 -90.2226 3.9116>
  }]
 pack [{shapeType=Capsule, length=0.505, radius=0.505, shapeName=pack_capsule, center=<0 0.2377 0.143>, rotation=<rot 90 0 0>}]
}
constraints [{
  bone1Name=Body_01
  bone2Name=pack
  swing1LimitDegrees=0
  swing1Motion=Locked
  swing2LimitDegrees=0
  swing2Motion=Locked
  twistLimitDegrees=14
 }]
