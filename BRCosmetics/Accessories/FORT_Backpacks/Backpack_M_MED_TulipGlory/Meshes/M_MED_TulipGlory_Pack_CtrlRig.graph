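# Control Rig for the TulipGlory backpack that procedurally animates the tail of its clawed creature skeleton.
# A reusable HarmonicCustom function turns a looping flip-flop timer plus noise into an eased angle, with flip and per-axis toggles, and rotates listed bones.
# Three calls rotate tail3, tail4 and tail5 on Z only at differing amplitudes.
# A Chain Harmonics unit then adds wave motion from tail1, with pendulum and reach off and debug drawing on.
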ControlRigBlueprint /BRCosmetics/Accessories/FORT_Backpacks/Backpack_M_MED_TulipGlory/Meshes/M_MED_TulipGlory_Pack_CtrlRig
graphs [
 HarmonicCustom_ContainedGraph {nodes [
   RigVMFunctionEntryNode()
   > For_Each = RigVMDispatchNode(Array=RigVMVariableNode() {variable=Items To Modify}.Value) {template="DISPATCH_RigVMDispatch_ArrayIterator(in Array,out Element,out Index,out Count,out Ratio)"}
   RigVMFunction_MathBoolNot = &RigVMFunction_MathBoolNot(Value=RigVMVariableNode() {variable=X Rot Disable}.Value)
   Remap = &RigVMFunction_MathDoubleRemap(
    SourceMaximum=1
    SourceMinimum=0
    TargetMaximum=-1
    TargetMinimum=1
    Value=&RigVMFunction_MathBoolToFloat(Value=RigVMVariableNode() {variable=FlipRotation}.Value).Result
   )
   Noise = /Script/RigVM.RigVMFunction_NoiseFloat(
    Maximum=0.1
    Minimum=0
    Frequency=RigVMVariableNode() {variable=Noise Frequency}.Value
    Speed=RigVMVariableNode() {variable=Noise speed}.Value
    Value=RigVMVariableNode() {variable=Noise Value}.Value
   ) {is=struct RigVMUnitNode, template="Noise::Execute(in Value,in Speed,in Frequency,in Minimum,in Maximum,out Result)"}
   RigVMFunction_TimeLoop = /Script/RigVM.RigVMFunction_TimeLoop(Normalize=False, Speed=RigVMVariableNode() {variable=TimeSpeed}.Value, Duration=RigVMVariableNode() {variable=TimeDuration}.Value) {is=struct RigVMUnitNode}
   Add = &RigVMFunction_MathDoubleAdd(B=@Noise.Result, A=@RigVMFunction_TimeLoop.FlipFlop)
   Cast = &RigVMFunction_MathBoolToFloat(Value=RigVMVariableNode() {variable=Enable Negative Angle}.Value)
   n1 = RigVMVariableNode() {variable=Angle}
   RigVMFunction_AnimEasing = /Script/RigVM.RigVMFunction_AnimEasing(
    SourceMaximum=1
    SourceMinimum=0
    Value=@Add.Result
    TargetMinimum=&RigVMFunction_MathDoubleMul(A=&RigVMFunction_MathDoubleMul(B=-1, A=@Cast.Result).Result, B=@n1.Value).Result
    TargetMaximum=@n1.Value
    Type=RigVMVariableNode() {variable=Ease Type Curve}.Value
   ) {is=struct RigVMUnitNode}
   n2 = &RigVMFunction_MathDoubleMul(B=@Remap.Result, A=@RigVMFunction_AnimEasing.Result)
   Multiply = &RigVMFunction_MathDoubleMul(B=&RigVMFunction_MathBoolToFloat(Value=@RigVMFunction_MathBoolNot.Result).Result, A=@n2.Result)
   n3 = &RigVMFunction_MathBoolNot(Value=RigVMVariableNode() {variable=Y Rot Disable}.Value)
   n4 = &RigVMFunction_MathBoolNot(Value=RigVMVariableNode() {variable=Z Rot Disable}.Value)
   RigVMFunction_MathQuaternionFromEuler = &RigVMFunction_MathQuaternionFromEuler(
    Euler=(X=0,Y=0,Z=0)
    Euler.X=@Multiply.Result
    Euler.Y=&RigVMFunction_MathDoubleMul(B=&RigVMFunction_MathBoolToFloat(Value=@n3.Result).Result, A=@n2.Result).Result
    Euler.Z=&RigVMFunction_MathDoubleMul(B=&RigVMFunction_MathBoolToFloat(Value=@n4.Result).Result, A=@n2.Result).Result
   )
   > &RigUnit_ModifyTransforms(
    ItemToModify=((Item=(Type=Bone,Name="upperarm_l"),Transform=(Rotation=(X=0,Y=0,Z=0,W=1),Translation=(X=0,Y=0,Z=0),Scale3D=(X=1,Y=1,Z=1))))
    Weight=1
    WeightMinimum=0
    ItemToModify.0.Item=@For_Each.Element
    ItemToModify.0.Transform.Rotation=@RigVMFunction_MathQuaternionFromEuler.Result
   )
   RigVMFunctionReturnNode(ExecuteContext=@For_Each.Completed)
  ]}
 RigVMFunctionLibrary {nodes [HarmonicCustom = RigVMCollapseNode(
    Angle=15
    Ease Type Curve=SineEaseInOut
    Enable Negative Angle=true
    ExecuteContext=()
    FlipRotation=false
    Items To Modify=((Type=Bone,Name="tail5"),(Type=Bone,Name="tail7"),(Type=Bone,Name="tail9"))
    Noise Frequency=0
    Noise Value=0
    Noise speed=1
    TimeDuration=1
    TimeSpeed=1
    X Rot Disable=false
    Y Rot Disable=false
    Z Rot Disable=false
   )]}
 RigVMModel {nodes [
   RigUnit_BeginExecution = &RigUnit_BeginExecution()
   HarmonicCustom_2 = RigVMFunctionReferenceNode(
    Angle=5
    Ease Type Curve=Linear
    Enable Negative Angle=false
    FlipRotation=false
    Items To Modify=((Type=Bone,Name="tail3"))
    Noise Frequency=0
    Noise Value=0
    Noise speed=1
    TimeDuration=2
    TimeSpeed=0.2
    X Rot Disable=true
    Y Rot Disable=true
    Z Rot Disable=false
    ExecuteContext=@RigUnit_BeginExecution.ExecutePin
   )
   > HarmonicCustom = RigVMFunctionReferenceNode(
    Angle=3
    Ease Type Curve=Linear
    Enable Negative Angle=false
    FlipRotation=false
    Items To Modify=((Type=Bone,Name="tail4"))
    Noise Frequency=0
    Noise Value=0
    Noise speed=1
    TimeDuration=2
    TimeSpeed=0.2
    X Rot Disable=true
    Y Rot Disable=true
    Z Rot Disable=false
   )
   > HarmonicCustom_1 = RigVMFunctionReferenceNode(
    Angle=7
    Ease Type Curve=Linear
    Enable Negative Angle=false
    FlipRotation=false
    Items To Modify=((Type=Bone,Name="tail5"))
    Noise Frequency=0
    Noise Value=0
    Noise speed=1
    TimeDuration=1
    TimeSpeed=0.2
    X Rot Disable=true
    Y Rot Disable=true
    Z Rot Disable=false
   )
   > /Script/ControlRig.RigUnit_ChainHarmonicsPerItem(
    ChainRoot=(Type=Bone,Name="tail1")
    DrawWorldOffset=(Rotation=(X=0,Y=0,Z=0,W=1),Translation=(X=0,Y=0,Z=0),Scale3D=(X=1,Y=1,Z=1))
    Pendulum=(bEnabled=false,PendulumStiffness=2,PendulumGravity=(X=0,Y=0,Z=0),PendulumBlend=0.75,PendulumDrag=0.98,PendulumMinimum=0,PendulumMaximum=0.1,PendulumEase=Linear,UnwindAxis=(X=0,Y=1,Z=0),UnwindMinimum=0.2,UnwindMaximum=0.05)
    Reach=(bEnabled=False,ReachTarget=(X=50,Y=0,Z=0),ReachAxis=(X=1,Y=0,Z=0),ReachMinimum=0,ReachMaximum=0,ReachEase=Linear)
    Speed=(X=0.8,Y=0.5,Z=0)
    Wave=(bEnabled=True,WaveFrequency=(X=0.3,Y=0.6,Z=0),WaveAmplitude=(X=0.3,Y=0.4,Z=0),WaveOffset=(X=0,Y=0.5,Z=0),WaveNoise=(X=0,Y=0,Z=0),WaveMinimum=0,WaveMaximum=1.2,WaveEase=ElasticEaseIn)
    WaveCurve=(EditorCurveData=(Keys=((),(InterpMode=RCIM_Cubic,Time=0.513,Value=0.5339,ArriveTangent=0.9029,LeaveTangent=0.9029),(InterpMode=RCIM_Cubic,TangentMode=RCTM_User,Time=1.0222,Value=0.9229,ArriveTangent=0.5016,LeaveTangent=0.5016))))
    bDrawDebug=True
   ) {is=struct RigVMUnitNode}
  ]}
]
hierarchy [
 root {loc=<0 0 0>, rot=<0 0 0 1>}
 pack {parent=root, loc=<0 -0.7316 0>, rot=<0 0 0 1>}
 body {parent=pack, loc=<0 -10.9505 -6.9239>, rot=<-0.0203 0 0 -0.9998>}
 container {parent=pack, loc=<-0.2851 -5.6486 -2.7728>, rot=<0 0 0 1>}
 Body_01 {parent=pack, loc=<0 -11.21 -7.5642>, rot=<-0.0143 0.708 0.0144 -0.7059>}
 claw2_r_1 {parent=Body_01, loc=<0.7884 2.0135 -3.0296>, rot=<-0.19 -0.6442 -0.3959 -0.6262>}
 claw2_r_2 {parent=claw2_r_1, loc=<3.1428 0 0>, rot=<0 0 -0.49 -0.8717>}
 claw2_r_3 {parent=claw2_r_2, loc=<2.7114 0.1474 -0.0849>, rot=<-0.0117 0.0047 -0.3745 -0.9271>}
 claw3_r_1 {parent=Body_01, loc=<-0.8413 1.809 -3.0717>, rot=<-0.173 -0.7622 -0.316 -0.5379>}
 claw3_r_2 {parent=claw3_r_1, loc=<3.018 0.0315 -0.242>, rot=<0 0 -0.4823 -0.876>}
 claw3_r_3 {parent=claw3_r_2, loc=<2.9496 -0.008 0.1051>, rot=<-0.0356 0.0129 -0.3417 -0.939>}
 claw4_r_1 {parent=Body_01, loc=<-2.6013 1.3525 -2.8581>, rot=<-0.1192 -0.8692 -0.2265 -0.4232>}
 claw4_r_2 {parent=claw4_r_1, loc=<3.3722 -0.0097 0>, rot=<0 0 -0.4817 -0.8763>}
 claw4_r_3 {parent=claw4_r_2, loc=<2.6652 -0.1863 0.051>, rot=<-0.0006 0.0002 -0.2927 -0.9562>}
 claw2_l_1 {parent=Body_01, loc=<0.8069 2.0135 3.0247>, rot=<0.1888 0.6422 -0.3965 -0.6282>}
 claw2_l_2 {parent=claw2_l_1, loc=<3.0166 -0.2075 0.0603>, rot=<0 0 -0.49 -0.8717>}
 claw2_l_3 {parent=claw2_l_2, loc=<2.9426 0.1901 -0.0048>, rot=<0.0117 -0.0047 -0.3745 -0.9271>}
 claw3_l_1 {parent=Body_01, loc=<-0.8225 1.809 3.0768>, rot=<0.172 0.7605 -0.3165 -0.5402>}
 claw3_l_2 {parent=claw3_l_1, loc=<2.9507 -0.1946 0.1642>, rot=<0 0 -0.4823 -0.876>}
 claw3_l_3 {parent=claw3_l_2, loc=<3.1708 -0.0137 0.001>, rot=<0.0356 -0.0129 -0.3417 -0.939>}
 claw4_l_1 {parent=Body_01, loc=<-2.5583 1.4888 3.0106>, rot=<-0.1186 -0.8679 0.2268 0.4258>}
 claw4_l_2 {parent=claw4_l_1, loc=<3.1571 -0.0578 0>, rot=<0 0 -0.4817 -0.8763>}
 claw4_l_3 {parent=claw4_l_2, loc=<2.7045 -0.2957 0.0004>, rot=<0.0006 -0.0002 -0.2927 -0.9562>}
 Body_03 {parent=Body_01, loc=<-2.6275 0.1802 0.008>, rot=<0 0 0 -1>}
 claw5_r_1 {parent=Body_03, loc=<-1.56 0.4545 -2.2748>, rot=<-0.1369 -0.9515 -0.1701 -0.2168>}
 claw5_r_2 {parent=claw5_r_1, loc=<3.2407 0.1585 0>, rot=<0 0 -0.4129 -0.9108>}
 claw5_r_3 {parent=claw5_r_2, loc=<2.5738 -0.0862 -0.0118>, rot=<-0.0664 0.0152 -0.2222 -0.9726>}
 claw5_l_1 {parent=Body_03, loc=<-1.5486 0.5594 2.2359>, rot=<-0.1369 -0.9515 0.17 0.2168>}
 claw5_l_2 {parent=claw5_l_1, loc=<3.2878 0.0547 0>, rot=<0 0 -0.4129 -0.9108>}
 claw5_l_3 {parent=claw5_l_2, loc=<2.4494 -0.0754 0.0103>, rot=<0.0664 -0.0152 -0.2222 -0.9726>}
 Body_02 {parent=Body_01, loc=<2.492 0.4293 -0.0076>, rot=<0.0003 -0.0017 -0.1462 -0.9893>}
 claw1_r_1 {parent=Body_02, loc=<0.7025 0.8902 -1.8902>, rot=<-0.3227 -0.5459 -0.1785 -0.7523>}
 claw1_r_2 {parent=claw1_r_1, loc=<2.2826 0 0>, rot=<0 0 -0.1686 -0.9857>}
 claw1_l_1 {parent=Body_02, loc=<0.7025 0.8902 1.8902>, rot=<0.3227 0.5459 -0.1785 -0.7523>}
 claw1_l_2 {parent=claw1_l_1, loc=<2.2826 0 0>, rot=<0 0 -0.1686 -0.9857>}
 tail1 {parent=Body_02, loc=<2.1054 0.2484 0>, rot=<0 0 0 1>}
 tail2 {parent=tail1, loc=<3.8256 -0.4733 0>, rot=<0 0 0.0981 -0.9952>}
 tail3 {parent=tail2, loc=<3.4993 0.2188 0.0169>, rot=<0.0236 -0.0114 -0.4341 -0.9005>}
 tail4 {parent=tail3, loc=<3.8868 0.1451 0.3695>, rot=<0.0044 -0.0021 -0.4347 -0.9006>}
 tail5 {parent=tail4, loc=<3.5051 0.6153 -0.1344>, rot=<0.1097 -0.0519 -0.4244 -0.8973>}
 tail6 {parent=tail5, loc=<2.8842 -0.016 -0.6361>, rot=<0.0719 -0.0284 -0.3668 -0.9271>}
 tail7 {parent=tail6, loc=<3.3353 -0.0245 0.1665>, rot=<0.6911 0.1498 0.1498 -0.6911>}
 tail8 {parent=tail7, loc=<3.4402 0.4062 -0.1096>, rot=<-0.3088 -0.0523 0.1587 -0.9363>}
 tail9 {parent=tail8, loc=<3.4407 0.3598 0.0369>, rot=<-0.3259 -0.0537 0.1534 -0.9313>}
 tail10 {parent=tail9, loc=<3.824 0.2777 0.0677>, rot=<0.1174 0.0525 0.3395 -0.9318>}
 tail11 {parent=tail10, loc=<1.7323 -0.3847 -0.0125>, rot=<-0.0532 -0.0035 0.4651 -0.8837>}
]
