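# Anim Blueprint for the TulipGlory backpack that plays an idle loop and offsets the pack bone as the wearer's spine bends.
# Each update maps the owner's spine_05-to-spine_01 yaw delta from ref pose to clamped pack rotation and translation, applied additively over the idle.
# A Control Rig, RigidBody and AnimDynamics chain for tail, body and claws exists but never reaches Output Pose.
# Its sine-driven motion logic hangs off an untriggered Sequence.
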AnimBlueprint /BRCosmetics/Accessories/FORT_Backpacks/Backpack_M_MED_TulipGlory/Meshes/M_MED_TulipGlory_Pack_AnimBP
target_skeleton=&/M_MED_TulipGlory_Pack_Skeleton
graphs [
 AnimGraph {nodes [
   Input Pose()
   n1 = "Transform (Modify) Bone - Bone: tail10"(Alpha=1, Translation="0,0,0", Rotation="0,-30,0", Scale="1,1,1") {
    rotationMode=BMM_Additive
    rotationSpace=BCS_BoneSpace
    alphaBoolBlend {blendOption=Linear}
    boneToModify {boneName=tail10}
    rotation {yaw=-30}
   }
   n2 = Control Rig(Source=Component To Local(ComponentPose=@n1)) {
    alphaBoolBlend {blendOption=Linear}
    controlRigAssetReference {blueprintRigClass=&/M_MED_TulipGlory_Pack_CtrlRig.M_MED_TulipGlory_Pack_CtrlRig_C}
   }
   n3 = "Transform (Modify) Bone - Bone: tail11"(ComponentPose=Local To Component(LocalPose=@n2), Alpha=0.7, Translation="0,0,0", Rotation="0,0,0", Scale="1,1,1") {rotationMode=BMM_Additive, rotationSpace=BCS_BoneSpace, alphaBoolBlend {blendOption=Linear}, boneToModify {boneName=tail11}}
   n4 = RigidBody(ComponentPose=@n3) {
    bOverrideWorldGravity=true
    overridePhysicsAsset=&/M_MED_TulipGlory_Pack_Body
    alphaBoolBlend {blendOption=Linear}
    componentAppliedLinearAccClamp=<6000 6000 3000>
    componentLinearVelScale=<4 4 1.5>
    overrideWorldGravity {z=-80}
    simSpaceSettings {maxAngularAcceleration=150, maxAngularVelocity=150, maxLinearAcceleration=300, maxLinearVelocity=300}
   }
   n5 = Anim Dynamics - Bone: Body_01(ComponentPose=@n4, Alpha=0.6, ExternalForce="0,0,0") {
    bLinearSpring=true
    bOverrideLinearDamping=true
    linearDampingOverride=1
    linearSpringConstant=800
    simulationSpace=RootRelative
    alphaBoolBlend {blendOption=Linear}
    boundBone {boneName=Body_01}
    physicsBodyDefinitions [{
      boundBone {boneName=Body_01}
      boxExtents=<50 50 50>
      constraintSetup {angularLimitsMax=<0 0 0>, angularLimitsMin=<0 0 0>, angularTarget=<0 0 0>, linearAxesMax=<0 0 0.5>, linearAxesMin=<0 0 -0.5>}
      localJointOffset=<0 0 0>
     }]
   }
   n6 = Anim Dynamics - Bone: claw1_r_1(ComponentPose=@n5, Alpha=1, ExternalForce="0,0,0") {
    angularDampingOverride=1
    angularSpringConstant=200
    bAngularSpring=true
    bOverrideAngularDamping=true
    bUseGravityOverride=true
    linearDampingOverride=0.9999
    linearSpringConstant=40
    simulationSpace=BoneRelative
    alphaBoolBlend {blendOption=Linear}
    boundBone {boneName=claw1_r_1}
    componentAppliedLinearAccClamp=<3000 3000 6000>
    componentLinearAccScale {y=0.3}
    componentLinearVelScale {y=0.3}
    gravityOverride {z=-100}
    physicsBodyDefinitions [{
      boundBone {boneName=claw1_r_1}
      boxExtents=<20 20 20>
      constraintSetup {angularLimitsMax=<0 0 45>, angularLimitsMin=<0 0 -5>, angularTarget=<1 0 0>}
      localJointOffset=<0 1 0>
     }]
    relativeSpaceBone {boneName=root}
    simSpaceSettings {maxAngularVelocity=3000}
   }
   n7 = Anim Dynamics - Bone: claw2_l_1(ComponentPose=@n6, Alpha=1, ExternalForce="0,0,0") {
    angularDampingOverride=1
    angularSpringConstant=450
    bAngularSpring=true
    bOverrideAngularDamping=true
    bUseGravityOverride=true
    linearDampingOverride=0.9999
    linearSpringConstant=40
    simulationSpace=BoneRelative
    alphaBoolBlend {blendOption=Linear}
    boundBone {boneName=claw2_l_1}
    componentAppliedLinearAccClamp=<3000 3000 6000>
    componentLinearAccScale {z=0.22}
    gravityOverride {z=-10}
    physicsBodyDefinitions [{
      boundBone {boneName=claw2_l_1}
      boxExtents=<20 20 20>
      constraintSetup {angularLimitsMax=<10 70 10>, angularLimitsMin=<0 -20 -5>, angularTarget=<1 0 0>}
      localJointOffset=<2.6 0 0>
     }]
    relativeSpaceBone {boneName=root}
    simSpaceSettings {maxAngularVelocity=3000}
   }
   n8 = Anim Dynamics - Bone: claw2_l_2(ComponentPose=@n7, Alpha=1, ExternalForce="0,0,0") {
    angularDampingOverride=0.9999
    angularSpringConstant=250
    bAngularSpring=true
    bOverrideAngularDamping=true
    bUseGravityOverride=true
    linearDampingOverride=0.9999
    linearSpringConstant=40
    simulationSpace=BoneRelative
    alphaBoolBlend {blendOption=Linear}
    boundBone {boneName=claw2_l_2}
    componentAppliedLinearAccClamp=<3000 3000 6000>
    componentLinearAccScale {y=0.35}
    componentLinearVelScale {y=0.35}
    gravityOverride {z=-100}
    physicsBodyDefinitions [{
      boundBone {boneName=claw2_l_2}
      boxExtents=<20 20 20>
      constraintSetup {angularLimitsMax=<0 0 60>, angularLimitsMin=<0 0 -5>, angularTarget=<1 0 0>}
      localJointOffset=<0 1.25 0>
     }]
    relativeSpaceBone {boneName=root}
    simSpaceSettings {maxAngularVelocity=3000}
   }
   n9 = Anim Dynamics - Bone: claw2_l_3(ComponentPose=@n8, Alpha=1, ExternalForce="0,0,0") {
    angularDampingOverride=1
    angularSpringConstant=200
    bAngularSpring=true
    bOverrideAngularDamping=true
    bUseGravityOverride=true
    linearDampingOverride=0.9999
    linearSpringConstant=40
    simulationSpace=BoneRelative
    alphaBoolBlend {blendOption=Linear}
    boundBone {boneName=claw2_l_3}
    componentAppliedLinearAccClamp=<3000 3000 6000>
    componentLinearAccScale {y=1}
    componentLinearVelScale {y=1}
    gravityOverride {z=-100}
    physicsBodyDefinitions [{
      boundBone {boneName=claw2_l_3}
      boxExtents=<20 20 20>
      constraintSetup {angularLimitsMax=<0 0 60>, angularLimitsMin=<0 0 -5>, angularTarget=<1 0 0>}
      localJointOffset=<0 1 0>
     }]
    relativeSpaceBone {boneName=root}
    simSpaceSettings {maxAngularVelocity=3000}
   }
   n10 = Anim Dynamics - Bone: claw3_l_1(ComponentPose=@n9, Alpha=1, ExternalForce="0,0,0") {
    angularDampingOverride=1
    angularSpringConstant=225
    bAngularSpring=true
    bOverrideAngularDamping=true
    bUseGravityOverride=true
    linearDampingOverride=0.9999
    linearSpringConstant=40
    simulationSpace=BoneRelative
    alphaBoolBlend {blendOption=Linear}
    boundBone {boneName=claw3_l_1}
    componentAppliedLinearAccClamp=<3000 3000 6000>
    componentLinearAccScale {z=0.16}
    gravityOverride {z=-10}
    physicsBodyDefinitions [{
      boundBone {boneName=claw3_l_1}
      boxExtents=<25 25 25>
      constraintSetup {angularLimitsMax=<10 35 10>, angularLimitsMin=<0 -30 -5>, angularTarget=<1 0 0>}
      localJointOffset=<1.5 0 0>
     }]
    relativeSpaceBone {boneName=root}
    simSpaceSettings {maxAngularVelocity=3000}
   }
   n11 = Anim Dynamics - Bone: claw3_l_2(ComponentPose=@n10, Alpha=1, ExternalForce="0,0,0") {
    angularDampingOverride=0.9999
    angularSpringConstant=250
    bAngularSpring=true
    bOverrideAngularDamping=true
    bUseGravityOverride=true
    linearDampingOverride=0.9999
    linearSpringConstant=40
    simulationSpace=BoneRelative
    alphaBoolBlend {blendOption=Linear}
    boundBone {boneName=claw3_l_2}
    componentAppliedLinearAccClamp=<3000 3000 6000>
    componentLinearAccScale {y=0.35}
    componentLinearVelScale {y=0.35}
    gravityOverride {z=-100}
    physicsBodyDefinitions [{
      boundBone {boneName=claw3_l_2}
      boxExtents=<20 20 20>
      constraintSetup {angularLimitsMax=<0 0 60>, angularLimitsMin=<0 0 -5>, angularTarget=<1 0 0>}
      localJointOffset=<0 1.25 0>
     }]
    relativeSpaceBone {boneName=root}
    simSpaceSettings {maxAngularVelocity=3000}
   }
   n12 = Anim Dynamics - Bone: claw3_l_3(ComponentPose=@n11, Alpha=1, ExternalForce="0,0,0") {
    angularDampingOverride=1
    angularSpringConstant=200
    bAngularSpring=true
    bOverrideAngularDamping=true
    bUseGravityOverride=true
    linearDampingOverride=0.9999
    linearSpringConstant=40
    simulationSpace=BoneRelative
    alphaBoolBlend {blendOption=Linear}
    boundBone {boneName=claw3_l_3}
    componentAppliedLinearAccClamp=<3000 3000 6000>
    componentLinearAccScale {y=1}
    componentLinearVelScale {y=1}
    gravityOverride {z=-100}
    physicsBodyDefinitions [{
      boundBone {boneName=claw3_l_3}
      boxExtents=<20 20 20>
      constraintSetup {angularLimitsMax=<0 0 60>, angularLimitsMin=<0 0 -5>, angularTarget=<1 0 0>}
      localJointOffset=<0 1 0>
     }]
    relativeSpaceBone {boneName=root}
    simSpaceSettings {maxAngularVelocity=3000}
   }
   n13 = Anim Dynamics - Bone: claw4_l_1(ComponentPose=@n12, Alpha=1, ExternalForce="0,0,0") {
    angularDampingOverride=1
    angularSpringConstant=225
    bAngularSpring=true
    bOverrideAngularDamping=true
    bUseGravityOverride=true
    linearDampingOverride=0.9999
    linearSpringConstant=40
    simulationSpace=BoneRelative
    alphaBoolBlend {blendOption=Linear}
    boundBone {boneName=claw4_l_1}
    componentAppliedLinearAccClamp=<3000 3000 6000>
    componentLinearAccScale {z=0.17}
    gravityOverride {z=-10}
    physicsBodyDefinitions [{
      boundBone {boneName=claw4_l_1}
      boxExtents=<20 20 20>
      constraintSetup {angularLimitsMax=<10 40 10>, angularLimitsMin=<0 -25 -5>, angularTarget=<1 0 0>}
      localJointOffset=<1.8 0 0>
     }]
    relativeSpaceBone {boneName=root}
    simSpaceSettings {maxAngularVelocity=3000}
   }
   n14 = Anim Dynamics - Bone: claw4_l_2(ComponentPose=@n13, Alpha=1, ExternalForce="0,0,0") {
    angularDampingOverride=0.9999
    angularSpringConstant=250
    bAngularSpring=true
    bOverrideAngularDamping=true
    bUseGravityOverride=true
    linearDampingOverride=0.9999
    linearSpringConstant=40
    simulationSpace=BoneRelative
    alphaBoolBlend {blendOption=Linear}
    boundBone {boneName=claw4_l_2}
    componentAppliedLinearAccClamp=<3000 3000 6000>
    componentLinearAccScale {y=0.35}
    componentLinearVelScale {y=0.35}
    gravityOverride {z=-100}
    physicsBodyDefinitions [{
      boundBone {boneName=claw4_l_2}
      boxExtents=<20 20 20>
      constraintSetup {angularLimitsMax=<0 0 60>, angularLimitsMin=<0 0 -5>, angularTarget=<1 0 0>}
      localJointOffset=<0 1.25 0>
     }]
    relativeSpaceBone {boneName=root}
    simSpaceSettings {maxAngularVelocity=3000}
   }
   n15 = Anim Dynamics - Bone: claw4_l_3(ComponentPose=@n14, Alpha=1, ExternalForce="0,0,0") {
    angularDampingOverride=1
    angularSpringConstant=200
    bAngularSpring=true
    bOverrideAngularDamping=true
    bUseGravityOverride=true
    linearDampingOverride=0.9999
    linearSpringConstant=40
    simulationSpace=BoneRelative
    alphaBoolBlend {blendOption=Linear}
    boundBone {boneName=claw4_l_3}
    componentAppliedLinearAccClamp=<3000 3000 6000>
    componentLinearAccScale {y=1}
    componentLinearVelScale {y=1}
    gravityOverride {z=-100}
    physicsBodyDefinitions [{
      boundBone {boneName=claw4_l_3}
      boxExtents=<20 20 20>
      constraintSetup {angularLimitsMax=<0 0 60>, angularLimitsMin=<0 0 -5>, angularTarget=<1 0 0>}
      localJointOffset=<0 1 0>
     }]
    relativeSpaceBone {boneName=root}
    simSpaceSettings {maxAngularVelocity=3000}
   }
   n16 = Anim Dynamics - Bone: claw5_l_1(ComponentPose=@n15, Alpha=1, ExternalForce="0,0,0") {
    angularDampingOverride=0.999
    angularSpringConstant=300
    bAngularSpring=true
    bOverrideAngularDamping=true
    bUseGravityOverride=true
    linearDampingOverride=0.9999
    linearSpringConstant=40
    simulationSpace=BoneRelative
    alphaBoolBlend {blendOption=Linear}
    boundBone {boneName=claw5_l_1}
    componentAppliedLinearAccClamp=<3000 3000 6000>
    componentLinearAccScale {x=2, z=0.9}
    componentLinearVelScale {x=2, z=0.9}
    gravityOverride {z=-100}
    physicsBodyDefinitions [{
      boundBone {boneName=claw5_l_1}
      boxExtents=<30 30 30>
      constraintSetup {angularLimitsMax=<0 30 45>, angularLimitsMin=<0 -20 0>, angularTarget=<1 0.2 0>}
      localJointOffset=<0.2 0.95 0>
     }]
    relativeSpaceBone {boneName=root}
    simSpaceSettings {maxAngularVelocity=3000}
   }
   n17 = Anim Dynamics - Bone: claw5_l_2(ComponentPose=@n16, Alpha=1, ExternalForce="0,0,0") {
    angularDampingOverride=0.99
    angularSpringConstant=250
    bAngularSpring=true
    bOverrideAngularDamping=true
    bUseGravityOverride=true
    linearDampingOverride=0.9999
    linearSpringConstant=40
    simulationSpace=BoneRelative
    alphaBoolBlend {blendOption=Linear}
    boundBone {boneName=claw5_l_2}
    componentAppliedLinearAccClamp=<3000 3000 6000>
    componentLinearAccScale {y=1, z=0.4}
    componentLinearVelScale {y=1, z=0.4}
    gravityOverride {z=-100}
    physicsBodyDefinitions [{
      boundBone {boneName=claw5_l_2}
      boxExtents=<30 30 30>
      constraintSetup {angularLimitsMax=<0 0 30>, angularLimitsMin=<0 0 -5>, angularTarget=<1 0 0>}
      localJointOffset=<0 0.5 0>
     }]
    relativeSpaceBone {boneName=root}
    simSpaceSettings {maxAngularVelocity=3000}
   }
   n18 = Anim Dynamics - Bone: claw1_l_1(ComponentPose=@n17, Alpha=1, ExternalForce="0,0,0") {
    angularDampingOverride=1
    angularSpringConstant=200
    bAngularSpring=true
    bOverrideAngularDamping=true
    bUseGravityOverride=true
    linearDampingOverride=0.9999
    linearSpringConstant=40
    simulationSpace=BoneRelative
    alphaBoolBlend {blendOption=Linear}
    boundBone {boneName=claw1_l_1}
    componentAppliedLinearAccClamp=<3000 3000 6000>
    componentLinearAccScale {y=0.3}
    componentLinearVelScale {y=0.3}
    gravityOverride {z=-100}
    physicsBodyDefinitions [{
      boundBone {boneName=claw1_l_1}
      boxExtents=<20 20 20>
      constraintSetup {angularLimitsMax=<0 0 45>, angularLimitsMin=<0 0 -5>, angularTarget=<1 0 0>}
      localJointOffset=<0 1 0>
     }]
    relativeSpaceBone {boneName=root}
    simSpaceSettings {maxAngularVelocity=3000}
   }
   n19 = Anim Dynamics - Bone: claw2_r_1(ComponentPose=@n18, Alpha=1, ExternalForce="0,0,0") {
    angularDampingOverride=1
    angularSpringConstant=450
    bAngularSpring=true
    bOverrideAngularDamping=true
    bUseGravityOverride=true
    linearDampingOverride=0.9999
    linearSpringConstant=40
    simulationSpace=BoneRelative
    alphaBoolBlend {blendOption=Linear}
    boundBone {boneName=claw2_r_1}
    componentAppliedLinearAccClamp=<3000 3000 6000>
    componentLinearAccScale {z=0.22}
    gravityOverride {z=-10}
    physicsBodyDefinitions [{
      boundBone {boneName=claw2_r_1}
      boxExtents=<20 20 20>
      constraintSetup {angularLimitsMax=<10 70 10>, angularLimitsMin=<0 -20 -5>, angularTarget=<1 0 0>}
      localJointOffset=<2.6 0 0>
     }]
    relativeSpaceBone {boneName=root}
    simSpaceSettings {maxAngularVelocity=3000}
   }
   n20 = Anim Dynamics - Bone: claw2_r_2(ComponentPose=@n19, Alpha=1, ExternalForce="0,0,0") {
    angularDampingOverride=0.9999
    angularSpringConstant=250
    bAngularSpring=true
    bOverrideAngularDamping=true
    bUseGravityOverride=true
    linearDampingOverride=0.9999
    linearSpringConstant=40
    simulationSpace=BoneRelative
    alphaBoolBlend {blendOption=Linear}
    boundBone {boneName=claw2_r_2}
    componentAppliedLinearAccClamp=<3000 3000 6000>
    componentLinearAccScale {y=0.35}
    componentLinearVelScale {y=0.35}
    gravityOverride {z=-100}
    physicsBodyDefinitions [{
      boundBone {boneName=claw2_r_2}
      boxExtents=<20 20 20>
      constraintSetup {angularLimitsMax=<0 0 60>, angularLimitsMin=<0 0 -5>, angularTarget=<1 0 0>}
      localJointOffset=<0 1.25 0>
     }]
    relativeSpaceBone {boneName=root}
    simSpaceSettings {maxAngularVelocity=3000}
   }
   n21 = Anim Dynamics - Bone: claw2_r_3(ComponentPose=@n20, Alpha=1, ExternalForce="0,0,0") {
    angularDampingOverride=1
    angularSpringConstant=200
    bAngularSpring=true
    bOverrideAngularDamping=true
    bUseGravityOverride=true
    linearDampingOverride=0.9999
    linearSpringConstant=40
    simulationSpace=BoneRelative
    alphaBoolBlend {blendOption=Linear}
    boundBone {boneName=claw2_r_3}
    componentAppliedLinearAccClamp=<3000 3000 6000>
    componentLinearAccScale {y=1}
    componentLinearVelScale {y=1}
    gravityOverride {z=-100}
    physicsBodyDefinitions [{
      boundBone {boneName=claw2_r_3}
      boxExtents=<20 20 20>
      constraintSetup {angularLimitsMax=<0 0 60>, angularLimitsMin=<0 0 -5>, angularTarget=<1 0 0>}
      localJointOffset=<0 1 0>
     }]
    relativeSpaceBone {boneName=root}
    simSpaceSettings {maxAngularVelocity=3000}
   }
   n22 = Anim Dynamics - Bone: claw3_r_1(ComponentPose=@n21, Alpha=1, ExternalForce="0,0,0") {
    angularDampingOverride=1
    angularSpringConstant=225
    bAngularSpring=true
    bOverrideAngularDamping=true
    bUseGravityOverride=true
    linearDampingOverride=0.9999
    linearSpringConstant=40
    simulationSpace=BoneRelative
    alphaBoolBlend {blendOption=Linear}
    boundBone {boneName=claw3_r_1}
    componentAppliedLinearAccClamp=<3000 3000 6000>
    componentLinearAccScale {z=0.16}
    gravityOverride {z=-10}
    physicsBodyDefinitions [{
      boundBone {boneName=claw3_r_1}
      boxExtents=<25 25 25>
      constraintSetup {angularLimitsMax=<10 35 10>, angularLimitsMin=<0 -30 -5>, angularTarget=<1 0 0>}
      localJointOffset=<1.5 0 0>
     }]
    relativeSpaceBone {boneName=root}
    simSpaceSettings {maxAngularVelocity=3000}
   }
   n23 = Anim Dynamics - Bone: claw3_r_2(ComponentPose=@n22, Alpha=1, ExternalForce="0,0,0") {
    angularDampingOverride=0.9999
    angularSpringConstant=250
    bAngularSpring=true
    bOverrideAngularDamping=true
    bUseGravityOverride=true
    linearDampingOverride=0.9999
    linearSpringConstant=40
    simulationSpace=BoneRelative
    alphaBoolBlend {blendOption=Linear}
    boundBone {boneName=claw3_r_2}
    componentAppliedLinearAccClamp=<3000 3000 6000>
    componentLinearAccScale {y=0.35}
    componentLinearVelScale {y=0.35}
    gravityOverride {z=-100}
    physicsBodyDefinitions [{
      boundBone {boneName=claw3_r_2}
      boxExtents=<20 20 20>
      constraintSetup {angularLimitsMax=<0 0 60>, angularLimitsMin=<0 0 -5>, angularTarget=<1 0 0>}
      localJointOffset=<0 1.25 0>
     }]
    relativeSpaceBone {boneName=root}
    simSpaceSettings {maxAngularVelocity=3000}
   }
   n24 = Anim Dynamics - Bone: claw3_r_3(ComponentPose=@n23, Alpha=1, ExternalForce="0,0,0") {
    angularDampingOverride=1
    angularSpringConstant=200
    bAngularSpring=true
    bOverrideAngularDamping=true
    bUseGravityOverride=true
    linearDampingOverride=0.9999
    linearSpringConstant=40
    simulationSpace=BoneRelative
    alphaBoolBlend {blendOption=Linear}
    boundBone {boneName=claw3_r_3}
    componentAppliedLinearAccClamp=<3000 3000 6000>
    componentLinearAccScale {y=1}
    componentLinearVelScale {y=1}
    gravityOverride {z=-100}
    physicsBodyDefinitions [{
      boundBone {boneName=claw3_r_3}
      boxExtents=<20 20 20>
      constraintSetup {angularLimitsMax=<0 0 60>, angularLimitsMin=<0 0 -5>, angularTarget=<1 0 0>}
      localJointOffset=<0 1 0>
     }]
    relativeSpaceBone {boneName=root}
    simSpaceSettings {maxAngularVelocity=3000}
   }
   n25 = Anim Dynamics - Bone: claw4_r_1(ComponentPose=@n24, Alpha=1, ExternalForce="0,0,0") {
    angularDampingOverride=1
    angularSpringConstant=225
    bAngularSpring=true
    bOverrideAngularDamping=true
    bUseGravityOverride=true
    linearDampingOverride=0.9999
    linearSpringConstant=40
    simulationSpace=BoneRelative
    alphaBoolBlend {blendOption=Linear}
    boundBone {boneName=claw4_r_1}
    componentAppliedLinearAccClamp=<3000 3000 6000>
    componentLinearAccScale {z=0.17}
    gravityOverride {z=-10}
    physicsBodyDefinitions [{
      boundBone {boneName=claw4_r_1}
      boxExtents=<20 20 20>
      constraintSetup {angularLimitsMax=<10 40 10>, angularLimitsMin=<0 -25 -5>, angularTarget=<1 0 0>}
      localJointOffset=<1.8 0 0>
     }]
    relativeSpaceBone {boneName=root}
    simSpaceSettings {maxAngularVelocity=3000}
   }
   n26 = Anim Dynamics - Bone: claw4_r_2(ComponentPose=@n25, Alpha=1, ExternalForce="0,0,0") {
    angularDampingOverride=0.9999
    angularSpringConstant=250
    bAngularSpring=true
    bOverrideAngularDamping=true
    bUseGravityOverride=true
    linearDampingOverride=0.9999
    linearSpringConstant=40
    simulationSpace=BoneRelative
    alphaBoolBlend {blendOption=Linear}
    boundBone {boneName=claw4_r_2}
    componentAppliedLinearAccClamp=<3000 3000 6000>
    componentLinearAccScale {y=0.35}
    componentLinearVelScale {y=0.35}
    gravityOverride {z=-100}
    physicsBodyDefinitions [{
      boundBone {boneName=claw4_r_2}
      boxExtents=<20 20 20>
      constraintSetup {angularLimitsMax=<0 0 60>, angularLimitsMin=<0 0 -5>, angularTarget=<1 0 0>}
      localJointOffset=<0 1.25 0>
     }]
    relativeSpaceBone {boneName=root}
    simSpaceSettings {maxAngularVelocity=3000}
   }
   n27 = Anim Dynamics - Bone: claw4_r_3(ComponentPose=@n26, Alpha=1, ExternalForce="0,0,0") {
    angularDampingOverride=1
    angularSpringConstant=200
    bAngularSpring=true
    bOverrideAngularDamping=true
    bUseGravityOverride=true
    linearDampingOverride=0.9999
    linearSpringConstant=40
    simulationSpace=BoneRelative
    alphaBoolBlend {blendOption=Linear}
    boundBone {boneName=claw4_r_3}
    componentAppliedLinearAccClamp=<3000 3000 6000>
    componentLinearAccScale {y=1}
    componentLinearVelScale {y=1}
    gravityOverride {z=-100}
    physicsBodyDefinitions [{
      boundBone {boneName=claw4_r_3}
      boxExtents=<20 20 20>
      constraintSetup {angularLimitsMax=<0 0 60>, angularLimitsMin=<0 0 -5>, angularTarget=<1 0 0>}
      localJointOffset=<0 1 0>
     }]
    relativeSpaceBone {boneName=root}
    simSpaceSettings {maxAngularVelocity=3000}
   }
   n28 = Anim Dynamics - Bone: claw5_r_1(ComponentPose=@n27, Alpha=1, ExternalForce="0,0,0") {
    angularDampingOverride=0.999
    angularSpringConstant=300
    bAngularSpring=true
    bOverrideAngularDamping=true
    bUseGravityOverride=true
    linearDampingOverride=0.9999
    linearSpringConstant=40
    simulationSpace=BoneRelative
    alphaBoolBlend {blendOption=Linear}
    boundBone {boneName=claw5_r_1}
    componentAppliedLinearAccClamp=<3000 3000 6000>
    componentLinearAccScale {x=2, z=0.9}
    componentLinearVelScale {x=2, z=0.9}
    gravityOverride {z=-100}
    physicsBodyDefinitions [{
      boundBone {boneName=claw5_r_1}
      boxExtents=<30 30 30>
      constraintSetup {angularLimitsMax=<0 30 45>, angularLimitsMin=<0 -20 0>, angularTarget=<1 0.2 0>}
      localJointOffset=<0.2 0.95 0>
     }]
    relativeSpaceBone {boneName=root}
    simSpaceSettings {maxAngularVelocity=3000}
   }
   n29 = Anim Dynamics - Bone: claw5_r_2(ComponentPose=@n28, Alpha=1, ExternalForce="0,0,0") {
    angularDampingOverride=0.99
    angularSpringConstant=250
    bAngularSpring=true
    bOverrideAngularDamping=true
    bUseGravityOverride=true
    linearDampingOverride=0.9999
    linearSpringConstant=40
    simulationSpace=BoneRelative
    alphaBoolBlend {blendOption=Linear}
    boundBone {boneName=claw5_r_2}
    componentAppliedLinearAccClamp=<3000 3000 6000>
    componentLinearAccScale {y=1, z=0.4}
    componentLinearVelScale {y=1, z=0.4}
    gravityOverride {z=-100}
    physicsBodyDefinitions [{
      boundBone {boneName=claw5_r_2}
      boxExtents=<30 30 30>
      constraintSetup {angularLimitsMax=<0 0 30>, angularLimitsMin=<0 0 -5>, angularTarget=<1 0 0>}
      localJointOffset=<0 0.5 0>
     }]
    relativeSpaceBone {boneName=root}
    simSpaceSettings {maxAngularVelocity=3000}
   }
   n30 = "Transform (Modify) Bone - Bone: Body_01"!(ComponentPose=@n29, Alpha=1, Translation=Get body move() -> (body move:Vector)) -> (Pose) {translationMode=BMM_Additive, translationSpace=BCS_BoneSpace, alphaBoolBlend {blendOption=Linear}, boneToModify {boneName=Body_01}}
   n31 = "Transform (Modify) Bone - Bone: claw1_l_1"!(ComponentPose=@n30, Alpha=1, Rotation="0,0,0") -> (Pose) {rotationMode=BMM_Additive, rotationSpace=BCS_BoneSpace, alphaBoolBlend {blendOption=Linear}, boneToModify {boneName=claw1_l_1}}
   n32 = "Transform (Modify) Bone - Bone: claw2_l_1"!(ComponentPose=@n31, Alpha=1, Rotation="0,0,0") -> (Pose) {rotationMode=BMM_Additive, rotationSpace=BCS_BoneSpace, alphaBoolBlend {blendOption=Linear}, boneToModify {boneName=claw2_l_1}}
   n33 = "Transform (Modify) Bone - Bone: claw3_l_1"!(ComponentPose=@n32, Alpha=1, Rotation="0,0,0") -> (Pose) {rotationMode=BMM_Additive, rotationSpace=BCS_BoneSpace, alphaBoolBlend {blendOption=Linear}, boneToModify {boneName=claw3_l_1}}
   n34 = "Transform (Modify) Bone - Bone: claw4_l_1"!(ComponentPose=@n33, Alpha=1, Rotation="0,0,0") -> (Pose) {rotationMode=BMM_Additive, rotationSpace=BCS_BoneSpace, alphaBoolBlend {blendOption=Linear}, boneToModify {boneName=claw4_l_1}}
   n35 = "Transform (Modify) Bone - Bone: claw5_l_1"!(ComponentPose=@n34, Alpha=1, Rotation="0,0,0") -> (Pose) {rotationMode=BMM_Additive, rotationSpace=BCS_BoneSpace, alphaBoolBlend {blendOption=Linear}, boneToModify {boneName=claw5_l_1}}
   n36 = "Transform (Modify) Bone - Bone: claw1_r_1"!(ComponentPose=@n35, Alpha=1, Rotation="0,0,0") -> (Pose) {rotationMode=BMM_Additive, rotationSpace=BCS_BoneSpace, alphaBoolBlend {blendOption=Linear}, boneToModify {boneName=claw1_r_1}}
   n37 = "Transform (Modify) Bone - Bone: claw2_r_1"!(ComponentPose=@n36, Alpha=1, Rotation="0,0,0") -> (Pose) {rotationMode=BMM_Additive, rotationSpace=BCS_BoneSpace, alphaBoolBlend {blendOption=Linear}, boneToModify {boneName=claw2_r_1}}
   n38 = "Transform (Modify) Bone - Bone: claw3_r_1"!(ComponentPose=@n37, Alpha=1, Rotation="0,0,0") -> (Pose) {rotationMode=BMM_Additive, rotationSpace=BCS_BoneSpace, alphaBoolBlend {blendOption=Linear}, boneToModify {boneName=claw3_r_1}}
   n39 = "Transform (Modify) Bone - Bone: claw4_r_1"!(ComponentPose=@n38, Alpha=1, Rotation="0,0,0") -> (Pose) {rotationMode=BMM_Additive, rotationSpace=BCS_BoneSpace, alphaBoolBlend {blendOption=Linear}, boneToModify {boneName=claw4_r_1}}
   n40 = "Transform (Modify) Bone - Bone: claw5_r_1"!(ComponentPose=@n39, Alpha=1, Rotation="0,0,0") -> (Pose) {rotationMode=BMM_Additive, rotationSpace=BCS_BoneSpace, alphaBoolBlend {blendOption=Linear}, boneToModify {boneName=claw5_r_1}}
   n41 = "Transform (Modify) Bone - Bone: tail1"(ComponentPose=@n40, Alpha=1, Translation="0,0,0", Rotation=Get move tail() -> (move tail:Rotator), Scale="1,1,1") {rotationMode=BMM_Additive, rotationSpace=BCS_BoneSpace, alphaBoolBlend {blendOption=Linear}, boneToModify {boneName=tail1}}
   "Transform (Modify) Bone - Bone: tail5"(ComponentPose=@n41, Alpha=1, Translation="0,0,0", Rotation=Get move tail_1() -> (move tail_1:Rotator), Scale="1,1,1") {rotationMode=BMM_Additive, rotationSpace=BCS_BoneSpace, alphaBoolBlend {blendOption=Linear}, boneToModify {boneName=tail5}}
   n42 = "Transform (Modify) Bone - Bone: pack"(
    ComponentPose=Local To Component(LocalPose=Sequence Player 'TulipGlory_Idle_PACK'() {sequence=/BRCosmetics/Accessories/FORT_Backpacks/Backpack_M_MED_TulipGlory/Animations/TulipGlory_Idle_PACK})
    Translation=Get backpack_translation()
    Rotation=Get backpack_rotate()
   ) {rotationMode=BMM_Additive, translationMode=BMM_Additive, alphaBoolBlend {blendOption=Linear}, boneToModify {boneName=pack}}
   Output Pose(Result=Component To Local(ComponentPose=@n42)) {layerGroup=None}
  ]}
 EventGraph {nodes [
   n1 = Event BlueprintUpdateAnimation()
   > n2 = Set Delta_Time_X(Delta_Time_X:float=@n1.DeltaTimeX) -> (Output_Get:float)
   n3 = GetOwnerSkeletalMeshComponent()
   n4 = BreakRotator(InRot=BreakTransform(InTransform=GetDeltaTransformFromRefPose(self=@n3, BoneName=spine_05, BaseName=spine_01)).Rotation)
   n5 = MapRangeClamped(Value=@n4.Yaw, InRangeA=-10, InRangeB=-50, OutRangeB=50)
   Set backpack_rotate(execute=[@n6.then_0, @n2.then], backpack_rotate=MakeRotator(Roll=@n5, Pitch=float / float(A=@n5, B=5)))
   > Set backpack_translation(backpack_translation=MakeVector(Y=MapRangeClamped(Value=@n4.Yaw, InRangeA=-15, InRangeB=-50, OutRangeB=-10)))
   n6 = Sequence!(execute) -> (then_0, then_1, then_2:Exec, then_3:Exec, then_4:Exec, then_5:Exec)
   n7 = GetRealTimeSeconds()
   n8 = "Sin (Radians)"(A=float * float(A=@n7, B=Get Foot_02_Speed_Multiplier() -> (Foot_02_Speed_Multiplier:double)) {is=title K2Node_PromotableOperator})
   n9 = Set Foot_01_Sin_float(
    execute=@n6.then_3
    Foot_01_Sin_float:double=FInterpTo(
     Current=Get Foot_01_Sin_float() -> (Foot_01_Sin_float:double)
     Target="Clamp (Float)"(Value=@n8, Min=0, Max=1)
     DeltaTime=Get Delta_Time_X() -> (Delta_Time_X:float)
     InterpSpeed=5
    )
   ) -> (Output_Get:double)
   n10 = "Sin (Radians)"(A=float * float(A=@n7, B=Get Foot_02_Speed_Multiplier()))
   > n11 = Set Foot_02_Sin_float(Foot_02_Sin_float:double=FInterpTo(
     Current=Get Foot_02_Sin_float() -> (Foot_02_Sin_float:double)
     Target="Clamp (Float)"(Value=@n10, Min=0, Max=1)
     DeltaTime=Get Delta_Time_X()
     InterpSpeed=15
    )) -> (Output_Get:double)
   n12 = "Sin (Radians)"(A=float * float(A=@n7, B=Get Foot_03_Speed_Multiplier() -> (Foot_03_Speed_Multiplier:double)))
   > n13 = Set Foot_03_Sin_float(Foot_03_Sin_float:double=FInterpTo(
     Current=Get Foot_03_Sin_float() -> (Foot_03_Sin_float:double)
     Target="Clamp (Float)"(Value=@n12, Min=0, Max=1)
     DeltaTime=Get Delta_Time_X()
     InterpSpeed=25
    )) -> (Output_Get:double)
   > Set Foot_01_Rot(Foot_01_Rot:Rotator=MakeRotator(Yaw=float * float(A=@n9, B=30))) -> (Output_Get:Rotator)
   > Set Foot_02_Rot(Foot_02_Rot:Rotator=MakeRotator(Yaw=float * float(A=@n11, B=35))) -> (Output_Get:Rotator)
   > Set Foot_03_Rot(Foot_03_Rot:Rotator=MakeRotator(Yaw=float * float(A=@n13, B=25))) -> (Output_Get:Rotator)
   n14 = Branch(execute=@n6.then_2, Condition=Get Foot_01_Animation_Enabled() -> (Foot_01_Animation_Enabled:bool))
   > Set Foot_01_Speed_Multiplier(Foot_01_Speed_Multiplier:double=2) -> (Output_Get:double)
   > Set Foot_02_Speed_Multiplier(Foot_02_Speed_Multiplier:double=4) -> (Output_Get:double)
   > Set Foot_03_Speed_Multiplier(Foot_03_Speed_Multiplier:double=6) -> (Output_Get:double)
   > Delay(Duration=RandomFloatInRange(Min=1, Max=3.5))
   > Set Foot_01_Animation_Enabled(Foot_01_Animation_Enabled:bool=false) -> (Output_Get:bool)
   Set Foot_01_Speed_Multiplier(execute=@n14.else, Foot_01_Speed_Multiplier=0)
   > Set Foot_02_Speed_Multiplier(Foot_02_Speed_Multiplier=0)
   > Set Foot_03_Speed_Multiplier(Foot_03_Speed_Multiplier=0)
   > Delay(Duration=RandomFloatInRange(Min=3, Max=6))
   > Set Foot_01_Animation_Enabled(Foot_01_Animation_Enabled=true)
   n15 = "Sin (Radians)"(A=float + float(A=float * float(A=@n7, B=0.45)))
   Set body move(execute=@n6.then_1, body move:Vector=MakeVector(X=float * float(A=@n15, B=1.1), Y=0)) -> (Output_Get:Vector)
   n16 = "Sin (Radians)"(A=float + float(A=float * float(A=@n7, B=2.5)))
   Set move tail(execute=@n6.then_4, move tail:Rotator=MakeRotator(Pitch=float * float(A=@n16, B=4.5))) -> (Output_Get:Rotator)
   n17 = "Sin (Radians)"(A=float + float(A=float * float(A=@n7, B=2.8)))
   > Set move tail_1(move tail_1:Rotator=MakeRotator(Pitch=float * float(A=@n17, B=6))) -> (Output_Get:Rotator)
   n18 = "Sin (Radians)"(A=float + float(A=float * float(A=@n7, B=1.4)))
   > Set move tail_3(move tail_3:Rotator=MakeRotator(Roll=float * float(A=@n18, B=11))) -> (Output_Get:Rotator)
   TryGetPawnOwner()
   Reroute Node(InputPin:Wildcard) -> (OutputPin:Wildcard)
   n19 = "Sin (Radians)"(A=float * float(B=Get Foot_02_Speed_Multiplier()))
   n20 = Set Foot_02_Sin_float(Foot_02_Sin_float=FInterpTo(Current=Get Foot_02_Sin_float(), Target="Clamp (Float)"(Value=@n19, Min=0, Max=1), DeltaTime=Get Delta_Time_X(), InterpSpeed=15))
   Set rot tail(rot tail:Rotator=MakeRotator(Yaw=float * float(A=@n20, B=-35))) -> (Output_Get:Rotator)
  ]}
]
variables [
 Alpha parachute {type=number}
 backpack_rotate {type=Rotator}
 backpack_translation {type=Vector}
 body move {type=Vector}
 Delta_Time_X {type=number}
 Foot_01_Animation_Enabled {type=boolean}
 Foot_01_Rot {type=Rotator}
 Foot_01_Sin_float {type=number}
 Foot_01_Speed_Multiplier {type=number}
 Foot_02_Animation_Enabled {type=boolean}
 Foot_02_Rot {type=Rotator}
 Foot_02_Sin_float {type=number}
 Foot_02_Speed_Multiplier {type=number}
 Foot_03_Rot {type=Rotator}
 Foot_03_Sin_float {type=number}
 Foot_03_Speed_Multiplier {type=number}
 get real time seconds {type=number}
 move tail {type=Rotator}
 move tail_1 {type=Rotator}
 move tail_3 {type=Rotator}
 Neck_Plates_02 {type=number}
 NewVar_2 {type=number}
 rot tail {type=Rotator}
]
